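# Niagara aura of soft looping sprite glows for the PistolCanary WealthLamb, built around a skeletal mesh.
# Glows and Glows001 follow 14 bone or socket indices on User.SkeletalMesh, layering grey material-animated puffs with rotating sprites randomly tinted blue to red.
# Glows002 and Glows003 repeat those looks without mesh sampling, growing with camera distance so the aura reads from afar.
# Bits and Em_Shaft, the only users of RarityColor, are disabled.
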NiagaraSystem /PistolCanary/Assets/FX/WealthLamb/Aura/NS_PistolCanary_WealthLamb_Aura
dependencies {
 usedDataInterfaces [
  &NiagaraDataInterfaceCurve
  &NiagaraDataInterfaceMeshRendererInfo
  &NiagaraDataInterfaceVector2DCurve
  &NiagaraDataInterfaceColorCurve
  &NiagaraDataInterfaceCamera
 ]
 usedDynamicInputs []
 usedModules [
  &self:SystemSpawnScript.SystemScriptSource.SystemScriptGraph.NiagaraNodeAssignment_0.SetVariables_#1
  &SystemState
 ]
}
emitters {
 Bits {
  bEnabled=false
  emitterUpdateScript [
   &EmitterState()
   &SpawnRate(SpawnRate=24, Use Spawn Probability=false)
  ]
  particleSpawnScript [
   &InitializeParticle(
    Color=&MultiplyLinearColor(
     Linear Color=$User.RarityColor
     Scale Factor=&MakeFloat4FromVectorAndFloat(
      "Vector (RGB)"=&VectorFromFloat(Value=&Lerp_Float(
        A=0.25
        B=1
        Alpha=&FloatFromCurve(
         FloatCurve=DataInterface() {propertyValues {Curve {keys [
             {interpMode=RCIM_Cubic, tangentMode=RCTM_User, time=0, value=0}
             {arriveTangent=0.8992, interpMode=RCIM_Cubic, leaveTangent=0.8992, tangentMode=RCTM_User, time=0.8, value=0.25}
             {arriveTangent=2.5764, interpMode=RCIM_Cubic, leaveTangent=2.5764, tangentMode=RCTM_User, time=1, value=1}
            ]}}}
         CurveIndex=&RandomRangeFloat(Minimum=0, Maximum=1)
        )
       ))
      "Float (Alpha)"=1
     )
    )
    Lifetime Max=0.85
    Lifetime Min=0.1
    Lifetime Mode=::NewEnumerator1
    Sprite Size Mode=::NewEnumerator3
    Uniform Sprite Size=&Lerp_Float(
     A=2.5
     B=8
     Alpha=&FloatFromCurve(
      FloatCurve=DataInterface() {propertyValues {Curve {keys [
          {interpMode=RCIM_Cubic, tangentMode=RCTM_User, time=0, value=0}
          {arriveTangent=0.8992, interpMode=RCIM_Cubic, leaveTangent=0.8992, tangentMode=RCTM_User, time=0.8, value=0.25}
          {arriveTangent=2.5764, interpMode=RCIM_Cubic, leaveTangent=2.5764, tangentMode=RCTM_User, time=1, value=1}
         ]}}}
      CurveIndex=&RandomRangeFloat(Minimum=0, Maximum=1)
     )
    )
    UsePositionOffset=false
    Write Material Random=true
   )
   &AddVelocity(
    Module.Use Falloff Distance=false
    Override Seed=false
    Velocity Mode=::NewEnumerator1
    Velocity Speed=&Lerp_Float(
     A=25
     B=100
     Alpha=&FloatFromCurve(
      FloatCurve=DataInterface() {propertyValues {Curve {keys [
          {interpMode=RCIM_Cubic, tangentMode=RCTM_User, time=0, value=0}
          {arriveTangent=0.8992, interpMode=RCIM_Cubic, leaveTangent=0.8992, tangentMode=RCTM_User, time=0.8, value=0.25}
          {arriveTangent=2.5764, interpMode=RCIM_Cubic, leaveTangent=2.5764, tangentMode=RCTM_User, time=1, value=1}
         ]}}}
      CurveIndex=&RandomRangeFloat(Minimum=0, Maximum=1)
     )
    )
   ) {enabled=false}
   &SkeletalMeshLocation(Bone Sampling Mode=::NewEnumerator0, MaskSpawnGroup=false, OverrideRandomSeed=false, Skeletal Mesh=$System.Skeletal Mesh)
   &ShapeLocation(
    MaskSpawnGroup=false
    Override Seed=false
    Shape Origin=$StackContext.SkeletalMeshLocation.SampledPosition
    Sphere Radius=40
    UseOwnerScale=true
   )
   &AddVelocity(Velocity=&MakeVector(X=0, Y=0, Z=&RandomRangeFloat(Minimum=15, Maximum=50))) {module=AddVelocity001}
  ]
  particleUpdateScript [
   &ParticleState()
   &Drag(Use Linear Drag=true, Use Rotational Drag=false) {enabled=false}
   &CurlNoiseForce(
    Noise Frequency=40
    Noise Strength=&FloatFromCurve(
     FloatCurve=DataInterface() {propertyValues {Curve {keys [
         {interpMode=RCIM_Cubic, leaveTangent=-2.6381, tangentMode=RCTM_Break, time=0.1, value=1}
         {interpMode=RCIM_Cubic, tangentMode=RCTM_Break, time=1, value=0}
        ]}}}
     Scale Curve=&RandomRangeFloat(Minimum=300, Maximum=1000)
    )
    Pan Noise=false
   ) {enabled=false}
   &SolveForcesAndVelocity(Clamp Velocity=false, Limit Acceleration=false)
   &ScaleColor(
    Scale Alpha=&FloatFromCurve(FloatCurve=DataInterface() {propertyValues {Curve {keys [{interpMode=RCIM_Cubic, tangentMode=RCTM_User, time=0, value=1}, {interpMode=RCIM_Cubic, tangentMode=RCTM_User, time=1, value=0}]}}})
    Scale RGB=&VectorFromFloat(Value=&Lerp_Float(
      A=1
      B=0.5
      Alpha=&FloatFromCurve(
       FloatCurve=DataInterface() {propertyValues {Curve {keys [
           {interpMode=RCIM_Cubic, tangentMode=RCTM_User, time=0, value=0}
           {arriveTangent=0.8992, interpMode=RCIM_Cubic, leaveTangent=0.8992, tangentMode=RCTM_User, time=0.8, value=0.25}
           {arriveTangent=2.5764, interpMode=RCIM_Cubic, leaveTangent=2.5764, tangentMode=RCTM_User, time=1, value=1}
          ]}}}
       CurveIndex=&RandomRangeFloat(Minimum=0, Maximum=1)
      )
     ))
    ScaleA=true
    ScaleRGB=true
   )
   &ScaleSpriteSize(Uniform Curve Sprite Scale=DataInterface() {propertyValues {Curve {keys [
        {interpMode=RCIM_Cubic, tangentMode=RCTM_User, time=0, value=0.5}
        {interpMode=RCIM_Cubic, tangentMode=RCTM_User, time=0.1, value=1}
        {interpMode=RCIM_Cubic, tangentMode=RCTM_Break, time=1, value=0.2}
       ]}}}) {module=ScaleSpriteSize001}
   &CameraOffset(Camera Offset Amount=15)
  ]
  rendererClasses [&NiagaraSpriteRendererProperties]
  renderers [{rendererClass=&NiagaraSpriteRendererProperties}]
 }
 Em_Shaft {
  bEnabled=false
  emitterUpdateScript [
   &EmitterState()
   &SpawnRate(SpawnRate=64, Use Spawn Probability=false)
  ]
  particleSpawnScript [
   /SetVariables(Particles.CurveRand:NiagaraFloat=&FloatFromCurve(
     FloatCurve=DataInterface() {propertyValues {Curve {keys [
         {interpMode=RCIM_Cubic, tangentMode=RCTM_User, time=0, value=0}
         {arriveTangent=0.7649, interpMode=RCIM_Cubic, leaveTangent=0.7649, tangentMode=RCTM_User, time=0.8, value=0.2}
         {arriveTangent=1.7117, interpMode=RCIM_Cubic, leaveTangent=1.7117, tangentMode=RCTM_User, time=1, value=1}
        ]}}}
     CurveIndex=&RandomRangeFloat(Minimum=0, Maximum=1)
    )) {isSetParameters=true}
   &InitializeParticle(
    Color=&MultiplyLinearColor(
     Linear Color=$User.RarityColor
     Scale Factor=&MakeFloat4FromVectorAndFloat(
      "Vector (RGB)"=&VectorFromFloat(Value=&Lerp_Float(
        A=0.25
        B=4
        Alpha=&FloatFromCurve(
         FloatCurve=DataInterface() {propertyValues {Curve {keys [
             {interpMode=RCIM_Cubic, tangentMode=RCTM_User, time=0, value=0}
             {arriveTangent=0.8992, interpMode=RCIM_Cubic, leaveTangent=0.8992, tangentMode=RCTM_User, time=0.8, value=0.25}
             {arriveTangent=2.5764, interpMode=RCIM_Cubic, leaveTangent=2.5764, tangentMode=RCTM_User, time=1, value=1}
            ]}}}
         CurveIndex=&RandomRangeFloat(Minimum=0, Maximum=1)
        )
       ))
      "Float (Alpha)"=1
     )
    )
    Lifetime Max=1
    Lifetime Min=0.5
    Lifetime Mode=::NewEnumerator1
    Sprite Rotation Mode=::NewEnumerator0
    Sprite Size=&Lerp_Vector2(A=<2.5 50>, B=<2.5 25>, Alpha=$Particles.CurveRand)
    Sprite Size Mode=::NewEnumerator4
    UsePositionOffset=false
    Write Material Random=true
   )
   &AddVelocity(
    Rotation Coordinate Space=::World
    Velocity=&Lerp_Vector(A=<0 0 25>, B=<0 0 2>, Alpha=$Particles.CurveRand)
    Velocity Speed Scale=3
   )
  ]
  particleUpdateScript [
   &SkeletalMeshLocation(Bone Sampling Mode=::NewEnumerator0, MaskSpawnGroup=false, OverrideRandomSeed=false, Skeletal Mesh=$System.Skeletal Mesh)
   &ShapeLocation(
    Cylinder Height=&Lerp_Float(A=30, B=10, Alpha=$Particles.CurveRand)
    Cylinder Radius=&Lerp_Float(A=10, B=45, Alpha=$Particles.CurveRand)
    MaskSpawnGroup=false
    Offset=<0 0 -30>
    Offset Mode=::NewEnumerator0
    Override Seed=false
    Shape Origin=$StackContext.SkeletalMeshLocation.SampledPosition
    Shape Primitive=::NewEnumerator1
    Surface Only=false
    UseOwnerScale=true
   )
   &ParticleState()
   &ScaleColor(
    Scale Alpha=&FloatFromCurve(
     FloatCurve=DataInterface() {propertyValues {Curve {keys [
         {interpMode=RCIM_Cubic, tangentMode=RCTM_Break, time=0, value=0}
         {interpMode=RCIM_Cubic, tangentMode=RCTM_Break, time=0.4, value=1}
         {interpMode=RCIM_Cubic, tangentMode=RCTM_Break, time=1, value=0}
        ]}}}
     Scale Curve=&RandomRangeFloat(Minimum=0.5, Maximum=0.9)
    )
    Scale RGB=&VectorFromFloat(Value=&Lerp_Float(
      A=1
      B=10
      Alpha=&FloatFromCurve(
       FloatCurve=DataInterface() {propertyValues {Curve {keys [
           {interpMode=RCIM_Cubic, tangentMode=RCTM_Break, time=0, value=0}
           {arriveTangent=0.3549, interpMode=RCIM_Cubic, leaveTangent=0.3549, tangentMode=RCTM_User, time=0.6, value=0.2}
           {arriveTangentWeight=0.1278, interpMode=RCIM_Cubic, tangentMode=RCTM_Break, time=0.9, value=1}
          ]}}}
       CurveIndex=&RandomRangeFloat(Minimum=0, Maximum=1)
      )
     ))
    ScaleA=true
    ScaleRGB=true
   )
   &ScaleSpriteSize(
    Scale Factor=&MakeVector2D(
     X=&FloatFromCurve(
      FloatCurve=DataInterface() {propertyValues {Curve {keys [{interpMode=RCIM_Cubic, tangentMode=RCTM_Break, time=0, value=0.1}, {interpMode=RCIM_Cubic, tangentMode=RCTM_Break, time=1, value=1}]}}}
      Scale Curve=/Niagara/DynamicInputs/Scalability/ScalabilityDistanceBasedFloat(
       Curve:NiagaraDataInterfaceCurve=DataInterface() {propertyValues {Curve {keys [
           {interpMode=RCIM_Cubic, tangentMode=RCTM_User, time=0, value=1}
           {arriveTangent=3.3503, interpMode=RCIM_Cubic, leaveTangent=3.3503, tangentMode=RCTM_User, time=0.7, value=2}
           {interpMode=RCIM_Cubic, tangentMode=RCTM_User, time=1, value=4}
          ]}}}
       CurveIndex:NiagaraFloat=$Emitter.Scalability.DistanceFraction
       Bias:NiagaraFloat=0
       Scale:NiagaraFloat=1
      )
     )
     Y=&FloatFromCurve(FloatCurve=DataInterface() {propertyValues {Curve {keys [
          {interpMode=RCIM_Cubic, leaveTangentWeight=0.5079, tangentMode=RCTM_Break, time=0, value=0.4}
          {interpMode=RCIM_Cubic, tangentMode=RCTM_Break, time=1, value=1}
         ]}}})
    )
    Scale Sprite Size Mode=::NewEnumerator2
   )
   &Drag(Drag=0.5, Use Linear Drag=true, Use Rotational Drag=true)
   &SolveForcesAndVelocity(Clamp Velocity=false, Limit Acceleration=false)
   &ScaleColor(
    Scale Alpha=&FNTimeOfDayScalar(Night Scale=0.8, Morning Scale=1, Evening Scale=1)
    ScaleA=true
    ScaleRGB=false
   ) {module=ScaleColor001}
  ]
  rendererClasses [&NiagaraSpriteRendererProperties]
  renderers [{rendererClass=&NiagaraSpriteRendererProperties}]
 }
 Glows {
  emitterUpdateScript [
   &EmitterState(Life Cycle Mode=::NewEnumerator1, Loop Behavior=::NewEnumerator1, Loop Duration Mode=::NewEnumerator1, UseLoopDelay=false)
   &SpawnRate(SpawnRate=6, Use Spawn Probability=false)
   &SpawnBurst_Instantaneous(Spawn Count=6, Use Spawn Probability=false, UseLoopCountLimit=false)
  ]
  particleSpawnScript [
   &InitializeParticle(
    Color=<rgba 0.1 0.1 0.1 0.1>
    Lifetime=&RandomRangeFloat(Minimum=2, Maximum=3)
    Sprite Size=<40 40>
    Sprite Size Mode=::NewEnumerator4
    UsePositionOffset=false
    Write Material Random=true
   )
   &FN_SetSpawnGroup(Number of Spawn Groups=14)
   &CameraOffset(Camera Offset Amount=10)
   &SkeletalMeshLocation(Bone / Socket Index=$Particles.SpawnGroup, Bone Sampling Mode=::NewEnumerator4, MaskSpawnGroup=false, Skeletal Mesh=$User.SkeletalMesh)
  ]
  particleUpdateScript [
   &ParticleState()
   &ScaleSpriteSize(
    Uniform Curve Scale=15
    Uniform Curve Sprite Scale=DataInterface() {propertyValues {Curve {keys [{tangentMode=RCTM_User, time=0, value=1}, {tangentMode=RCTM_User, time=1, value=2}]}}}
   )
   &SkeletalMeshLocation(Bone / Socket Index=$Particles.SpawnGroup, Bone Sampling Mode=::NewEnumerator4, MaskSpawnGroup=false, Skeletal Mesh=$User.SkeletalMesh) {module=SkeletalMeshLocation001}
   &SpriteRotationRate(Rotation Rate=&RandomRangeFloat(Minimum=-50, Maximum=50)) {enabled=false}
   &DynamicMaterialParameters(
    Index 0 Param 1=&Add_Float(
     A=&RandomRangeFloat(Minimum=0, Maximum=999)
     B=&FloatFromCurve(
      FloatCurve=DataInterface() {propertyValues {Curve {
         defaultValue=0
         keys [
          {arriveTangent=4.1785, interpMode=RCIM_Cubic, leaveTangent=4.1785, tangentMode=RCTM_User, time=0, value=0}
          {arriveTangent=1.1179, interpMode=RCIM_Cubic, leaveTangent=1.1179, tangentMode=RCTM_User, time=0.2042, value=0.6228}
          {arriveTangent=0.0224, interpMode=RCIM_Cubic, leaveTangent=0.0224, tangentMode=RCTM_User, time=1, value=1}
         ]
        }}}
      Scale Curve=2
     )
    )
    Index 0 Param 2=&FloatFromCurve(FloatCurve=DataInterface() {propertyValues {Curve {
        defaultValue=-3.75
        keys [
         {interpMode=RCIM_Cubic, tangentMode=RCTM_User, time=0, value=-3}
         {interpMode=RCIM_Cubic, tangentMode=RCTM_User, time=0.5, value=0}
         {tangentMode=RCTM_User, time=1, value=-3}
        ]
       }}})
    Module.Index0Float0Write=true
    Module.Index0Float1Write=true
    Module.Index0Float2Write=true
    Module.Index0Float3Write=true
   )
   &ScaleColor(
    Scale Alpha=&FloatFromCurve(FloatCurve=DataInterface() {propertyValues {Curve {
        defaultValue=1
        keys [
         {interpMode=RCIM_Cubic, tangentMode=RCTM_User, time=0, value=0}
         {interpMode=RCIM_Cubic, tangentMode=RCTM_User, time=0.5, value=1}
         {tangentMode=RCTM_User, time=1, value=0}
        ]
       }}})
    Scale RGB=&VectorFromFloat(Value=&FloatFromCurve(FloatCurve=DataInterface() {propertyValues {Curve {
         defaultValue=1
         keys [
          {interpMode=RCIM_Cubic, tangentMode=RCTM_User, time=0, value=0}
          {interpMode=RCIM_Cubic, tangentMode=RCTM_User, time=0.5, value=1}
          {tangentMode=RCTM_User, time=1, value=0}
         ]
        }}}))
    ScaleA=true
    ScaleRGB=true
   )
   &GravityForce(Gravity=<0 0 20>)
   &SolveForcesAndVelocity(Clamp Velocity=false, Limit Acceleration=false)
  ]
  rendererClasses [&NiagaraSpriteRendererProperties]
  renderers [{rendererClass=&NiagaraSpriteRendererProperties}]
 }
 Glows001 {
  emitterUpdateScript [
   &EmitterState(Life Cycle Mode=::NewEnumerator1, Loop Behavior=::NewEnumerator1, Loop Duration Mode=::NewEnumerator1, UseLoopDelay=false)
   &SpawnRate(SpawnRate=6, Use Spawn Probability=false)
   &SpawnBurst_Instantaneous(Spawn Count=14, Use Spawn Probability=false, UseLoopCountLimit=false) {enabled=false}
  ]
  particleSpawnScript [
   &InitializeParticle(
    Color=&ColorFromCurve(
     ColorCurve=DataInterface() {propertyValues {
       AlphaCurve {keys [<at 0 0.5 RCIM_Cubic>,<at 1 0.5 RCIM_Cubic>]}
       BlueCurve {keys [<at 0 1 RCIM_Cubic>,<at 1 0 RCIM_Cubic>]}
       GreenCurve {keys [<at 0 0.3001 RCIM_Cubic>,<at 1 0 RCIM_Cubic>]}
       RedCurve {keys [<at 0 0.02 RCIM_Cubic>,<at 1 1 RCIM_Cubic>]}
      }}
     CurveIndex=&RandomRangeFloat(Minimum=0, Maximum=1)
    )
    Sprite Rotation Mode=::NewEnumerator2
    Sprite Size=<300 300>
    Sprite Size Mode=::NewEnumerator4
    UsePositionOffset=false
    Write Material Random=true
   )
   &FN_SetSpawnGroup(Number of Spawn Groups=14)
   &CameraOffset()
   &SkeletalMeshLocation(Bone / Socket Index=$Particles.SpawnGroup, Bone Sampling Mode=::NewEnumerator4, MaskSpawnGroup=false, Skeletal Mesh=$User.SkeletalMesh)
  ]
  particleUpdateScript [
   &ParticleState()
   &SkeletalMeshLocation(Bone / Socket Index=$Particles.SpawnGroup, Bone Sampling Mode=::NewEnumerator4, MaskSpawnGroup=false, Skeletal Mesh=$User.SkeletalMesh) {module=SkeletalMeshLocation001}
   &SpriteRotationRate(Rotation Rate=&RandomRangeFloat(Minimum=-50, Maximum=50))
   &ScaleColor(
    Linear Color Curve=DataInterface() {propertyValues {
      AlphaCurve {keys [<at 0 0>,<at 0.3189 0.9497>,<at 1 0>]}
      BlueCurve {keys [<at 0 1>,<at 0.136 1>,<at 1 1>]}
      GreenCurve {keys [<at 0 1>,<at 0.136 1>,<at 1 1>]}
      RedCurve {keys [<at 0 1>,<at 0.136 1>,<at 1 1>]}
     }}
    Scale Mode=::NewEnumerator2
   )
  ]
  rendererClasses [&NiagaraSpriteRendererProperties]
  renderers [{rendererClass=&NiagaraSpriteRendererProperties}]
 }
 Glows002 {
  emitterUpdateScript [
   &EmitterState(Life Cycle Mode=::NewEnumerator1, Loop Behavior=::NewEnumerator1, Loop Duration Mode=::NewEnumerator1, UseLoopDelay=false)
   &SpawnRate(Use Spawn Probability=false)
   &SpawnBurst_Instantaneous(Spawn Count=6, Use Spawn Probability=false, UseLoopCountLimit=false)
  ]
  particleSpawnScript [
   &InitializeParticle(
    Color=<rgba 0.2188 0.2188 0.2188 0.1739>
    Lifetime=&RandomRangeFloat(Minimum=2, Maximum=3)
    Sprite Size=<50 50>
    Sprite Size Mode=::NewEnumerator4
    UsePositionOffset=false
    Write Material Random=true
   )
   &CameraOffset(Camera Offset Amount=10)
  ]
  particleUpdateScript [
   &ParticleState()
   &ScaleSpriteSize(
    Uniform Curve Scale=&RemapRange(
     Input Value To Remap=/Niagara/DynamicInputs/Camera/DistanceToCamera(Use Owner Position:NiagaraBool=true, Camera Query:NiagaraDataInterfaceCamera=DataInterface())
     Input Min=5000
     Input Max=10000
     Output Min=0
     Output Max=8
    )
    Uniform Curve Sprite Scale=DataInterface() {propertyValues {Curve {keys [{tangentMode=RCTM_User, time=0, value=1}, {tangentMode=RCTM_User, time=1, value=2}]}}}
   )
   &SpriteRotationRate(Rotation Rate=&RandomRangeFloat(Minimum=-50, Maximum=50)) {enabled=false}
   &DynamicMaterialParameters(
    Index 0 Param 1=&Add_Float(
     A=&RandomRangeFloat(Minimum=0, Maximum=999)
     B=&FloatFromCurve(
      FloatCurve=DataInterface() {propertyValues {Curve {
         defaultValue=0
         keys [
          {arriveTangent=4.1785, interpMode=RCIM_Cubic, leaveTangent=4.1785, tangentMode=RCTM_User, time=0, value=0}
          {arriveTangent=1.1179, interpMode=RCIM_Cubic, leaveTangent=1.1179, tangentMode=RCTM_User, time=0.2042, value=0.6228}
          {arriveTangent=0.0224, interpMode=RCIM_Cubic, leaveTangent=0.0224, tangentMode=RCTM_User, time=1, value=1}
         ]
        }}}
      Scale Curve=2
     )
    )
    Index 0 Param 2=&FloatFromCurve(FloatCurve=DataInterface() {propertyValues {Curve {
        defaultValue=-3.75
        keys [
         {interpMode=RCIM_Cubic, tangentMode=RCTM_User, time=0, value=-3}
         {interpMode=RCIM_Cubic, tangentMode=RCTM_User, time=0.5, value=0}
         {tangentMode=RCTM_User, time=1, value=-3}
        ]
       }}})
    Module.Index0Float0Write=true
    Module.Index0Float1Write=true
    Module.Index0Float2Write=true
    Module.Index0Float3Write=true
   )
   &ScaleColor(
    Scale Alpha=&FloatFromCurve(FloatCurve=DataInterface() {propertyValues {Curve {
        defaultValue=1
        keys [
         {interpMode=RCIM_Cubic, tangentMode=RCTM_User, time=0, value=0}
         {interpMode=RCIM_Cubic, tangentMode=RCTM_User, time=0.5, value=1}
         {tangentMode=RCTM_User, time=1, value=0}
        ]
       }}})
    Scale RGB=&VectorFromFloat(Value=&FloatFromCurve(FloatCurve=DataInterface() {propertyValues {Curve {
         defaultValue=1
         keys [
          {interpMode=RCIM_Cubic, tangentMode=RCTM_User, time=0, value=0}
          {interpMode=RCIM_Cubic, tangentMode=RCTM_User, time=0.5, value=1}
          {tangentMode=RCTM_User, time=1, value=0}
         ]
        }}}))
    ScaleA=true
    ScaleRGB=true
   )
   &GravityForce(Gravity=<0 0 4>)
   &SolveForcesAndVelocity(Clamp Velocity=false, Limit Acceleration=false)
  ]
  rendererClasses [&NiagaraSpriteRendererProperties]
  renderers [{rendererClass=&NiagaraSpriteRendererProperties}]
 }
 Glows003 {
  emitterUpdateScript [
   &EmitterState(Life Cycle Mode=::NewEnumerator1, Loop Behavior=::NewEnumerator1, Loop Duration Mode=::NewEnumerator1, UseLoopDelay=false)
   &SpawnRate(SpawnRate=2, Use Spawn Probability=false)
  ]
  particleSpawnScript [
   &InitializeParticle(
    Color=&ColorFromCurve(
     ColorCurve=DataInterface() {propertyValues {
       AlphaCurve {keys [<at 0 1 RCIM_Cubic>,<at 1 1 RCIM_Cubic>]}
       BlueCurve {keys [<at 0 1 RCIM_Cubic>,<at 1 0 RCIM_Cubic>]}
       GreenCurve {keys [<at 0 0.3001 RCIM_Cubic>,<at 1 0 RCIM_Cubic>]}
       RedCurve {keys [<at 0 0.02 RCIM_Cubic>,<at 1 1 RCIM_Cubic>]}
      }}
     CurveIndex=&RandomRangeFloat(Minimum=0, Maximum=1)
    )
    Sprite Rotation Mode=::NewEnumerator2
    Sprite Size=<300 300>
    Sprite Size Mode=::NewEnumerator4
    UsePositionOffset=false
    Write Material Random=true
   )
   &CameraOffset(Camera Offset Amount=50)
  ]
  particleUpdateScript [
   &ParticleState()
   &ScaleSpriteSize(
    Scale Sprite Size Mode=::NewEnumerator0
    Uniform Scale Factor=&RemapRange(
     Input Value To Remap=&DistanceToCamera(Use Owner Position:NiagaraBool=true, Camera Query:NiagaraDataInterfaceCamera=DataInterface())
     Input Min=2000
     Input Max=10000
     Output Min=0
     Output Max=30
    )
   )
   &SpriteRotationRate(Rotation Rate=&RandomRangeFloat(Minimum=-50, Maximum=50))
   &ScaleColor(
    Linear Color Curve=DataInterface() {propertyValues {
      AlphaCurve {keys [<at 0 0>,<at 0.3189 0.9497>,<at 1 0>]}
      BlueCurve {keys [<at 0 1>,<at 0.136 1>,<at 1 1>]}
      GreenCurve {keys [<at 0 1>,<at 0.136 1>,<at 1 1>]}
      RedCurve {keys [<at 0 1>,<at 0.136 1>,<at 1 1>]}
     }}
    Scale Mode=::NewEnumerator2
   )
  ]
  rendererClasses [&NiagaraSpriteRendererProperties]
  renderers [{rendererClass=&NiagaraSpriteRendererProperties}]
 }
}
summary {
 systemName=NS_PistolCanary_WealthLamb_Aura
 emitters [
  {bEnabled=false, emitterName=Em_Shaft, rendererClasses [&NiagaraSpriteRendererProperties]}
  {bEnabled=false, emitterName=Bits, rendererClasses [&NiagaraSpriteRendererProperties]}
  {emitterName=Glows, rendererClasses [&NiagaraSpriteRendererProperties]}
  {emitterName=Glows001, rendererClasses [&NiagaraSpriteRendererProperties]}
  {emitterName=Glows002, rendererClasses [&NiagaraSpriteRendererProperties]}
  {emitterName=Glows003, rendererClasses [&NiagaraSpriteRendererProperties]}
 ]
 userVariables [
  User.RarityColor {type=LinearColor, description="", defaultValue {struct=LinearColor, value=<rgba 2 1.0388 0 1>}}
  User.Scale {type=NiagaraFloat, defaultValue=1, description=""}
  User.SkeletalMesh {
   type=NiagaraDataInterfaceSkeletalMesh
   defaultValue {
    struct=NiagaraExt_VariableValue_DataInterface
    value {dataInterface=&self:NiagaraDataInterfaceSkeletalMesh_0}
   }
  }
 ]
}
